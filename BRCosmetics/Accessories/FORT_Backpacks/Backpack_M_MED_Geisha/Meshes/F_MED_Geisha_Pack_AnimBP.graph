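# Anim Blueprint for the Geisha backpack cosmetic that adds spring-driven secondary motion to its dynamic bones.
# The output path runs the input pose through damped, angle-limited Anim Dynamics nodes for the jaw, then the left and right horn-eye bones.
# A second chain for the left and right horn a, b and c bones is not connected to the output, so it has no effect.
# The event graph holds no logic.
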AnimBlueprint /BRCosmetics/Accessories/FORT_Backpacks/Backpack_M_MED_Geisha/Meshes/F_MED_Geisha_Pack_AnimBP
target_skeleton=&/F_MED_Geisha_Pack_Skeleton
graphs [
 AnimGraph {nodes [
   n1 = Anim Dynamics - Bone: dyn_horn_l_a(Alpha=1, ExternalForce="0,0,0") {
    angularDampingOverride=0.9999
    angularSpringConstant=500
    bAngularSpring=true
    bGravityOverrideInSimSpace=true
    bOverrideAngularDamping=true
    bOverrideLinearDamping=true
    gravityScale=0.35
    simulationSpace=RootRelative
    alphaBoolBlend {blendOption=Linear}
    boundBone {boneName=dyn_horn_l_a}
    componentAppliedLinearAccClamp=<1500 1500 1500>
    componentLinearAccScale=<0.25 0.25 0.5>
    componentLinearVelScale=<0.25 0.25 0.5>
    physicsBodyDefinitions [{
      sphereCollisionRadius=0
      boundBone {boneName=dyn_horn_l_a}
      boxExtents=<15 15 5>
      constraintSetup {twistAxis=AxisZ, angularLimitsMax=<15 25 0>, angularLimitsMin=<-15 -25 0>, angularTarget=<0 0 0>}
      localJointOffset=<0 0 -2>
     }]
   }
   n2 = Anim Dynamics - Bone: dyn_horn_r_a(ComponentPose=@n1, Alpha=1, ExternalForce="0,0,0") {
    angularDampingOverride=0.9999
    angularSpringConstant=500
    bAngularSpring=true
    bGravityOverrideInSimSpace=true
    bOverrideAngularDamping=true
    bOverrideLinearDamping=true
    gravityScale=0.35
    simulationSpace=RootRelative
    alphaBoolBlend {blendOption=Linear}
    boundBone {boneName=dyn_horn_r_a}
    componentAppliedLinearAccClamp=<1500 1500 1500>
    componentLinearAccScale=<0.5 0.5 1>
    componentLinearVelScale=<0.25 0.25 0.5>
    physicsBodyDefinitions [{
      sphereCollisionRadius=0
      boundBone {boneName=dyn_horn_r_a}
      boxExtents=<15 15 5>
      constraintSetup {twistAxis=AxisZ, angularLimitsMax=<15 25 0>, angularLimitsMin=<-15 -25 0>, angularTarget=<0 0 0>}
      localJointOffset=<0 0 1.5>
     }]
   }
   n3 = Anim Dynamics - Bone: dyn_horn_l_b(ComponentPose=@n2, Alpha=1, ExternalForce="0,0,0") {
    angularDampingOverride=0.9999
    angularSpringConstant=500
    bAngularSpring=true
    bGravityOverrideInSimSpace=true
    bOverrideAngularDamping=true
    bOverrideLinearDamping=true
    gravityScale=0.35
    alphaBoolBlend {blendOption=Linear}
    boundBone {boneName=dyn_horn_l_b}
    componentAppliedLinearAccClamp=<100 100 100>
    componentLinearAccScale=<1 1 1>
    componentLinearVelScale=<1 1 1>
    physicsBodyDefinitions [{
      sphereCollisionRadius=0
      boundBone {boneName=dyn_horn_l_b}
      boxExtents=<10 10 5>
      constraintSetup {angularTargetAxis=AxisY, twistAxis=AxisZ, angularLimitsMax=<0 15 5>, angularLimitsMin=<0 -15 -5>, angularTarget=<0 0 0>}
      localJointOffset=<0 0 -1>
     }]
   }
   n4 = Anim Dynamics - Bone: dyn_horn_r_b(ComponentPose=@n3, Alpha=1, ExternalForce="0,0,0") {
    angularDampingOverride=0.9999
    angularSpringConstant=500
    bAngularSpring=true
    bGravityOverrideInSimSpace=true
    bOverrideAngularDamping=true
    bOverrideLinearDamping=true
    gravityScale=0.35
    alphaBoolBlend {blendOption=Linear}
    boundBone {boneName=dyn_horn_r_b}
    componentAppliedLinearAccClamp=<100 100 100>
    componentLinearAccScale=<1 1 1>
    componentLinearVelScale=<1 1 1>
    physicsBodyDefinitions [{
      sphereCollisionRadius=0
      boundBone {boneName=dyn_horn_r_b}
      boxExtents=<10 10 5>
      constraintSetup {angularTargetAxis=AxisY, twistAxis=AxisZ, angularLimitsMax=<0 15 5>, angularLimitsMin=<0 -15 -5>, angularTarget=<0 0 0>}
      localJointOffset=<0 0 1>
     }]
   }
   n5 = Anim Dynamics - Bone: dyn_horn_l_c(ComponentPose=@n4, Alpha=1, ExternalForce="0,0,0") {
    angularDampingOverride=1
    angularSpringConstant=500
    bAngularSpring=true
    bGravityOverrideInSimSpace=true
    bOverrideAngularDamping=true
    bOverrideLinearDamping=true
    gravityScale=0.35
    simulationSpace=RootRelative
    alphaBoolBlend {blendOption=Linear}
    boundBone {boneName=dyn_horn_l_c}
    componentAppliedLinearAccClamp=<100 100 100>
    componentLinearAccScale=<0.5 0.5 0.5>
    componentLinearVelScale=<0.5 0.5 0.5>
    physicsBodyDefinitions [{
      sphereCollisionRadius=0
      boundBone {boneName=dyn_horn_l_c}
      boxExtents=<10 10 5>
      constraintSetup {angularTargetAxis=AxisY, twistAxis=AxisZ, angularLimitsMax=<5 10 0>, angularLimitsMin=<-5 -5 0>, angularTarget=<0 0 0>}
      localJointOffset=<0 0 -3>
     }]
   }
   Anim Dynamics - Bone: dyn_horn_r_c(ComponentPose=@n5, Alpha=1, ExternalForce="0,0,0") {
    angularDampingOverride=1
    angularSpringConstant=500
    bAngularSpring=true
    bGravityOverrideInSimSpace=true
    bOverrideAngularDamping=true
    bOverrideLinearDamping=true
    gravityScale=0.35
    simulationSpace=RootRelative
    alphaBoolBlend {blendOption=Linear}
    boundBone {boneName=dyn_horn_r_c}
    componentAppliedLinearAccClamp=<100 100 100>
    componentLinearAccScale=<0.5 0.5 0.5>
    componentLinearVelScale=<0.5 0.5 0.5>
    physicsBodyDefinitions [{
      sphereCollisionRadius=0
      boundBone {boneName=dyn_horn_r_c}
      boxExtents=<10 10 5>
      constraintSetup {angularTargetAxis=AxisY, twistAxis=AxisZ, angularLimitsMax=<5 10 0>, angularLimitsMin=<-5 -5 0>, angularTarget=<0 0 0>}
      localJointOffset=<0 0 2>
     }]
   }
   AnimGraphNode_SubInput_0 = Input Pose()
   n6 = Anim Dynamics - Bone: dyn_jaw(ComponentPose=Local To Component(LocalPose=@AnimGraphNode_SubInput_0), Alpha=1, ExternalForce="0,0,0") {
    angularDampingOverride=0.9999
    angularSpringConstant=500
    bAngularSpring=true
    bGravityOverrideInSimSpace=true
    bOverrideAngularDamping=true
    bOverrideLinearDamping=true
    gravityScale=0.35
    simulationSpace=RootRelative
    alphaBoolBlend {blendOption=Linear}
    boundBone {boneName=dyn_jaw}
    componentAppliedLinearAccClamp=<500 500 500>
    componentLinearAccScale=<0.25 0.25 0.5>
    componentLinearVelScale=<0.5 0.5 0.5>
    physicsBodyDefinitions [{
      sphereCollisionRadius=0
      boundBone {boneName=dyn_jaw}
      boxExtents=<15 15 10>
      constraintSetup {twistAxis=AxisZ, angularLimitsMax=<7 0 5>, angularLimitsMin=<-5 0 -5>, angularTarget=<0 0 0>}
      localJointOffset=<0 0 -5>
     }]
   }
   n7 = Anim Dynamics - Bone: dyn_horn_l_eye(ComponentPose=@n6, Alpha=1, ExternalForce="0,0,0") {
    angularDampingOverride=1
    angularSpringConstant=500
    bAngularSpring=true
    bGravityOverrideInSimSpace=true
    bOverrideAngularDamping=true
    bOverrideLinearDamping=true
    gravityScale=0.35
    alphaBoolBlend {blendOption=Linear}
    boundBone {boneName=dyn_horn_l_eye}
    componentAppliedLinearAccClamp=<500 500 500>
    componentLinearAccScale=<1 0.25 0.25>
    componentLinearVelScale=<1 0.25 0.25>
    physicsBodyDefinitions [{
      sphereCollisionRadius=0
      boundBone {boneName=dyn_horn_l_eye}
      boxExtents=<10 10 5>
      constraintSetup {twistAxis=AxisY, angularLimitsMax=<20 0 20>, angularLimitsMin=<-20 0 -20>, angularTarget=<0 0 0>}
      localJointOffset=<0 0 -4>
     }]
   }
   n8 = Anim Dynamics - Bone: dyn_horn_r_eye(ComponentPose=@n7, Alpha=1, ExternalForce="0,0,0") {
    angularDampingOverride=1
    angularSpringConstant=500
    bAngularSpring=true
    bGravityOverrideInSimSpace=true
    bOverrideAngularDamping=true
    bOverrideLinearDamping=true
    gravityScale=0.35
    alphaBoolBlend {blendOption=Linear}
    boundBone {boneName=dyn_horn_r_eye}
    componentAppliedLinearAccClamp=<500 500 500>
    componentLinearAccScale=<1 0.25 0.25>
    componentLinearVelScale=<1 0.25 0.25>
    physicsBodyDefinitions [{
      sphereCollisionRadius=0
      boundBone {boneName=dyn_horn_r_eye}
      boxExtents=<10 10 5>
      constraintSetup {twistAxis=AxisY, angularLimitsMax=<20 0 20>, angularLimitsMin=<-20 0 -20>, angularTarget=<0 0 0>}
      localJointOffset=<0 0 4>
     }]
   }
   Output Pose(Result=Component To Local(ComponentPose=@n8)) {layerGroup=None}
  ]}
 EventGraph {nodes [
   TryGetPawnOwner()
   Event BlueprintUpdateAnimation()
  ]}
]
variables []
